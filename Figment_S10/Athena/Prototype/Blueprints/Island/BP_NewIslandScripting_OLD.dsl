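; Old prototype island scripting Blueprint that positions island foundations along a spline according to a calendar event's progress.
; After playlist data and authority match readiness, it reads the event, using its time ratio as Progress when active.
; GetIslandPosition samples the spline at that fraction of its length and grid-snaps the result.
; Foundations are dynamically placed there with a fixed yaw and enabled, while MapPingLocation's OnRep moves the map ping collision sphere.
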(event EventBeginPlay
  (bind _returnvalue (CallFunction|GetIslandPosition))
  (Class|FortAsyncActionCurrentPlaylistReady|PlaylistDataReadyAsync
    (:OnReady
      (Class|FortAsyncActionAuthorityMatchReady|MatchReadyAsync
        (:OnReady
          (bind (_biseventactive _timesincebegin _timeuntilend _timespanratio) (CalendarEvents|GetCalendarEventInformation (Variables|Default|GetCalendarEvent)))
          (if _biseventactive
            (Variables|Default|SetProgress _timespanratio)
            ; OLD
            (for _array_element (Variables|Default|GetIslandFoundation)
              (Variables|Default|SetMapPingLocation _returnvalue)
              (Class|BuildingFoundation|SetDynamicFoundationTransform _array_element (Math|Transform|MakeTransform _returnvalue "0,-89.999992, 0"))
              (Class|BuildingFoundation|SetDynamicFoundationEnabled _array_element true))))))))

(fn GetIslandPosition
  (return (Math|Vector|MakeVector (Math|Float|SnapToGrid(Float) (.x (Spline|GetLocationatDistanceAlongSpline (Variables|Default|GetSpline) (* (Spline|GetSplineLength (Variables|Default|GetSpline)) (Variables|Default|GetProgress)) "World")) 512.0) (Math|Float|SnapToGrid(Float) (.y (Spline|GetLocationatDistanceAlongSpline (Variables|Default|GetSpline) (* (Spline|GetSplineLength (Variables|Default|GetSpline)) (Variables|Default|GetProgress)) "World")) 512.0) (Math|Float|SnapToGrid(Float) (.z (Spline|GetLocationatDistanceAlongSpline (Variables|Default|GetSpline) (* (Spline|GetSplineLength (Variables|Default|GetSpline)) (Variables|Default|GetProgress)) "World")) 384.0))))

(fn OnRep_MapPingLocation
  (Transformation|SetWorldLocation (Variables|Default|GetMapPingCollisionSphere) (Variables|Default|GetMapPingLocation) false false))

(fn ConstructionScript)
